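# Material function that produces the animated plasma color gradient for the Cube Queen character, grouped under her emissive parameters.
# A scalar gradient input, defaulting to the red channel of a noise texture, is offset by time multiplied by a NoiseSpeed input.
# The fractional part of that sum indexes the emissive curve in the shared character curve atlas, cycling the gradient through the noise values without moving UVs.
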MaterialFunction /Game/Characters/Player/Female/Medium/Bodies/F_MED_Cube_Queen/Materials/MF_Cube_QueenPlasma
graph {nodes [
  n1 = TextureSample() {
   samplerType=SAMPLERTYPE_LinearGrayscale
   texture=/Game/Characters/Player/Female/Medium/Bodies/F_MED_Cube_Queen/Textures/TechArt/T_CubeQueenNoiseA
  }
  n2 = FunctionInput(Preview=@n1.R) {bUsePreviewValueAsDefault=true, inputName=GradientTexture, inputType=FunctionInput_Scalar}
  n3 = FunctionInput() {bUsePreviewValueAsDefault=true, inputName=NoiseSpeed, inputType=FunctionInput_Scalar, previewValue=<0.2 0 0 1>}
  n4 = Multiply(A=Time(), B=@n3)
  n5 = CurveAtlasRowParameter(CurveTime=Frac(None=Add(A=@n2, B=@n4))) {
   atlas=/Game/Characters/Shared/CurveAtlas/AthenaCharacterCurveAtlas
   curve=/Game/Characters/Shared/CurveAtlas/CQ_EmissiveCurve
   defaultValue=25
   group=Cube Queen Emissive
   parameterName=NoiseGradient
  }
  FunctionOutput(None=@n5)
 ]}
