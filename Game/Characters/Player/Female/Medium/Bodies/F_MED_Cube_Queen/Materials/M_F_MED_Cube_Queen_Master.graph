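# Master character material for the Cube Queen player body, wrapping switchable skin effects between the shared character PreFX and PostFX functions.
# A skin mask routes silk, clear coat, gem, thin-film, metal LUT and fresnel-driven anodized metal layers.
# Scrolling noise feeds plasma emissives, shaped by pre-skinned position masks for shoulders and tails.
# Vertex red indexes a curve atlas to offset floating pieces, with sine drift and Blueprint-timed deploy blending.
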Material /Game/Characters/Player/Female/Medium/Bodies/F_MED_Cube_Queen/Materials/M_F_MED_Cube_Queen_Master
graph {
 properties {
  bUseMaterialAttributes=true
  bUsedWithMeshDeformer=true
  bUsedWithMorphTargets=true
  bUsedWithSkeletalMesh=true
  ditherOpacityMask=true
  pixelDepthOffsetMode=PDOM_Legacy
  previewMesh=/Game/Gadgets/Assets/VinderTech_GliderChute/Glider_Cube_Queen_Female/Meshes/temp/Cube_Queen_Female_glider_Combined
  shadingModel=MSM_FromMaterialExpression
  subsurfaceProfile=&SS_HeroSkin_02
  usedShadingModels=Default Lit | Clear Coat | Subsurface Profile
 }
 nodes [
  n1 = PixelNormalWS()
  n2 = ComponentMask(None=NamedRerouteUsage()) {b=true, r=true}
  n3 = ScalarParameter() {defaultValue=1, group=Anodized Metal, parameterName=Anodized_Distortion_UVScale}
  n4 = Time()
  n5 = ScalarParameter() {defaultValue=0.007, group=Anodized Metal, parameterName=Anodized_Distortion_Speed}
  n6 = TextureSampleParameter2D(UVs=Add(A=Divide(A=@n2, B=@n3), B=Multiply(A=@n4, B=@n5))) {
   group=Anodized Metal
   parameterName=Anodized_NormalDistortion
   samplerType=SAMPLERTYPE_Normal
   texture=/Game/Global/Textures/Default/Noise/T_Noise_Normal_01
  }
  n7 = ScalarParameter() {group=Anodized Metal, parameterName=Anodized_Distortion}
  n8 = AppendVector(A=AppendVector(A=@n7, B=@n7), B=Constant() {r=1})
  n9 = MaterialFunctionCall(BaseNormal=@n1, AdditionalNormal=Normalize(VectorInput=Multiply(A=@n6.RGB, B=@n8))) {materialFunction=&BlendAngleCorrectedNormals}
  n10 = Add(
   A=Fresnel(
    ExponentIn=ScalarParameter() {defaultValue=1, group=Anodized Metal, parameterName=Anodized_FresnelExponent}
    BaseReflectFractionIn=ScalarParameter() {defaultValue=3, group=Anodized Metal, parameterName=Anodized_ReflectFraction}
    Normal=StaticSwitchParameter(True=@n9.Result, False=@n1) {group=Anodized Metal, parameterName=UseDistortion}
   ) {baseReflectFraction=0.5, exponent=3}
   B=ScalarParameter() {group=Anodized Metal, parameterName=Anodized_ColorOffset}
  )
  Add(B=@n10)
  n11 = MaterialFunctionCall() {materialFunction=&BaseFN_CharacterFunction_PreFX}
  n12 = MaterialFunctionCall!(In=TextureObjectParameter() {group=Skin FX, parameterName=SkinFX_Mask}) {materialFunction=/Game/Weapons/WeaponSkins/MaterialFunction/MF_GlitchSubfunction}
  n13 = StaticSwitchParameter(True=@n12.Glitched, False=@n12.Glitched) {group=Skin FX, parameterName=UseGlitch}
  n14 = StaticSwitchParameter(
   True=MaterialFunctionCall!(Alpha=ChannelMaskParameter(None=@n13) {group=Skin FX, parameterName=Silk Channel}.R, Material=@n11.Result) {materialFunction=/Game/Weapons/WeaponSkins/MaterialFunction/MF_Silk}.Result
   False=@n11.Result
  ) {group=Skin FX, parameterName=Use Silk}
  n15 = StaticSwitchParameter(
   True=MaterialFunctionCall!(Alpha=StaticComponentMaskParameter(None=@n13) {defaultR=true, group=Skin FX, parameterName=Clear Coat Channel}, In=@n14) {materialFunction=/Game/Weapons/WeaponSkins/MaterialFunction/Clearcoat}.Result
   False=@n14
  ) {group=Skin FX, parameterName=Use Clear Coat}
  n16 = StaticSwitchParameter(
   True=MaterialFunctionCall!(
    MASK=StaticComponentMaskParameter(None=@n13) {defaultR=true, group=Skin FX, parameterName=Generic Gem Effect Channel}
    Color Change Sequence=StaticComponentMaskParameter(None=@n13) {defaultB=true, group=Skin FX, parameterName=Generic Gem Color Change Channel}
    Material=@n15
   ) {materialFunction=/Game/Weapons/WeaponSkins/MaterialFunction/GenericGem}.Result
   False=@n15
  ) {group=Skin FX, parameterName=Use Generic Gem}
  n17 = StaticSwitchParameter(
   True=MaterialFunctionCall!(Alpha=StaticComponentMaskParameter(None=@n13) {defaultR=true, group=Skin FX, parameterName=ThinFilm_Channel}, Material Attributes=@n16) {materialFunction=/Game/Weapons/WeaponSkins/MaterialFunction/ThinFilmFunction}.Result
   False=@n16
  ) {group=Skin FX, parameterName=Use Thin Film}
  n18 = MaterialFunctionCall(Mask=@n13, Materials=@n17) {materialFunction=&MF_MetalsLUT}
  n19 = GetMaterialAttributes(None=MaterialFunctionCall!(In=StaticSwitchParameter(True=@n18.Result, False=@n17) {group=Skin FX, parameterName=UseMetalLUT}) {materialFunction=/Game/Weapons/WeaponSkins/MaterialFunction/MF_EmissiveFresnel}.Result) {attributeGetTypes [#1, #2, #3]}
  n20 = OneMinus(None=NamedRerouteUsage())
  n21 = Saturate(None=Add(A=@n20, B=ScalarParameter() {defaultValue=1, group=Cube Queen Emissive, parameterName=CubeQueenBaseColorMultiplier}))
  n22 = NamedRerouteUsage()
  n23 = ScalarParameter() {group=Anodized Metal, parameterName=AnodizedEffectIntensity}
  n24 = Add(
   A=@n19.Base Color
   B=StaticSwitchParameter(True=Multiply(A=@n22, B=@n23) {constB=0.1}, False=Constant()) {group=Skin FX, parameterName=Use Anodized Metal}
  )
  n25 = StaticSwitchParameter(True=LinearInterpolate(A=@n19.Emissive Color, B=NamedRerouteUsage(), Alpha=NamedRerouteUsage()), False=@n19.Emissive Color) {group=Cube Queen Emissive, parameterName=Use Cube Queen Emissive}
  n26 = NamedRerouteUsage()
  n27 = ComponentMask(None=NamedRerouteUsage()) {g=true}
  n28 = VectorParameter() {group=Cube Queen Emissive, parameterName=SecondaryEmissiveColorA, defaultValue=<rgba 0.0226 0.0593 0.28 0>}
  n29 = VectorParameter() {group=Cube Queen Emissive, parameterName=SecondaryEmissiveColorB, defaultValue=<rgba 0 1.2 0.6594 0>}
  n30 = Saturate(None=InverseLinearInterpolate(
    A=ScalarParameter() {defaultValue=0.4, group=Cube Queen Emissive, parameterName=SecondaryEmissiveGradientStart}
    B=ScalarParameter() {defaultValue=0.6, group=Cube Queen Emissive, parameterName=SecondaryEmissiveGradientEnd}
    Value=NamedRerouteUsage()
   ))
  n31 = Multiply(
   A=Multiply(A=Multiply(A=@n26, B=@n27), B=LinearInterpolate(A=@n28.RGB, B=@n29.RGB, Alpha=@n30))
   B=Power(Base=NamedRerouteUsage(), Exp=ScalarParameter() {defaultValue=2, group=Cube Queen Emissive, parameterName=SecondaryEmissiveFresnelPow})
  )
  n32 = Saturate(None=Add(A=@n20, B=ScalarParameter() {defaultValue=1, group=Cube Queen Emissive, parameterName=CubeQueenSpecularMultiplier}))
  n33 = SetMaterialAttributes!(
   MaterialAttributes=@n19.MaterialAttributes
   Base Color=StaticSwitchParameter(True=Multiply(A=@n21, B=@n24), False=@n24) {group=Cube Queen Emissive, parameterName=Use Cube Queen Emissive}
   Emissive Color=StaticSwitchParameter(True=Add(A=@n25, B=@n31), False=@n25) {group=Cube Queen Emissive, parameterName=UseSecondaryEmissives}
   World Position Offset=StaticSwitchParameter(True=NamedRerouteUsage(), False=Constant()) {group=Skin FX, parameterName=Use Cube Queen WPO}
   Specular=StaticSwitchParameter(True=Multiply(A=@n32, B=@n19.Specular), False=@n19.Specular) {group=Cube Queen Emissive, parameterName=Use Cube Queen Emissive}
  ) {attributeSetTypes [
    #1
    #2
    #4
    #3
   ]}
  MaterialFunctionCall(AO=@n11.AO, Distance Blend=@n11.AO, Fresnel=@n11.AO, Gmap=@n11.AO, Pre Skinned Local Position=@n11.AO, Material=@n33) {materialFunction=&BaseFN_CharacterFunction_PostFX}
  n34 = Multiply(
   A=TextureSample(UVs=@n10) {texture=/Game/Global/Textures/Default/Masks_and_Gradients/3pxRGBImage}.RGB
   B=VectorParameter() {group=Anodized Metal, parameterName=AnodizationFilter, defaultValue=<rgba 1 0 0.5305 0>}.RGB
  )
  n35 = StaticComponentMaskParameter(None=AtmosphericLightVector()) {
   defaultB=true
   defaultG=true
   defaultR=true
   desc=Use RG for 2D vectors dot product. else use RGB
   group=Anodized Metal
   parameterName=AnodizedLdotR_Components
  }
  n36 = StaticComponentMaskParameter(None=ReflectionVectorWS()) {
   defaultB=true
   defaultG=true
   defaultR=true
   desc=Use RG for 2D vectors dot product. else use RGB
   group=Anodized Metal
   parameterName=AnodizedLdotR_Components
  }
  n37 = Power(Base=DotProduct(A=@n35, B=@n36), Exp=ScalarParameter() {defaultValue=55, group=Anodized Metal, parameterName=Anodized_pow})
  n38 = Constant() {r=512}
  n39 = LinearInterpolate(A=@n34, B=Multiply(A=Constant() {r=4}, B=@n34), Alpha=Saturate(None=Multiply(A=@n37, B=@n38)))
  MaterialFunctionCall(Hue Shift Percentage=ScalarParameter() {group=Anodized Metal, parameterName=Anodized_Hue}, Texture=@n39) {materialFunction=&HueShift}
  NamedRerouteDeclaration(None=@n11.AO) {name=AO, nodeColor=<rgba 1 0 0.8437 1>}
  n40 = StaticComponentMaskParameter(None=@n13) {defaultG=true, group=Cube Queen Emissive, parameterName=CubeQueenPlasmaEmissiveEffect Channel}
  n41 = StaticComponentMaskParameter(None=@n13) {defaultG=true, group=Cube Queen Emissive, parameterName=CubeQueenPlasmaEmissiveEffect Channel2}
  NamedRerouteDeclaration(None=StaticSwitchParameter(True=Max(A=@n40, B=@n41), False=@n40) {group=Cube Queen Emissive, parameterName=UseTwoPlasmaEmissiveEffectChannels}) {name=CubeQueenPlasmaEmissiveChannel, nodeColor=<rgba 0.6875 0 1 1>}
  n42 = Multiply(
   A=Desaturation(None=@n39, Fraction=ScalarParameter() {group=Anodized Metal, parameterName=Anodized_Desat})
   B=Power(Base=NamedRerouteUsage(), Exp=ScalarParameter() {group=Anodized Metal, parameterName=AnodizedEffectFresnelMask})
  )
  n43 = NamedRerouteUsage()
  n44 = Constant()
  NamedRerouteDeclaration(None=ShadingPathSwitch(Default=Multiply(A=@n42, B=@n43), Forward=@n44, Mobile=@n44)) {name=AnodizedMetalColors, nodeColor=<rgba 1 0 0.4219 1>}
  NamedRerouteDeclaration(None=StaticComponentMaskParameter(None=@n13) {defaultB=true, group=Cube Queen Emissive, parameterName=CubeQueenEmissiveEffect Channel}) {name=CubeQueenEmissiveChannel, nodeColor=<rgba 0 1 0.5781 1>}
  n45 = TextureCoordinate()
  n46 = ComponentMask(None=@n45) {r=true}
  n47 = Sign(None=ComponentMask(None=NamedRerouteUsage()) {r=true})
  n48 = Multiply(A=@n47, B=ScalarParameter() {defaultValue=-0.25, group=Cube Queen Emissive, parameterName=UV_SymmetryOffset})
  n49 = LinearInterpolate(A=@n46, B=OneMinus(None=@n46), Alpha=Saturate(None=@n47))
  n50 = AppendVector(A=LinearInterpolate(A=@n46, B=Add(A=@n48, B=@n49), Alpha=NamedRerouteUsage()), B=ComponentMask(None=@n45) {g=true})
  n51 = TextureSampleParameter2D!(
   UVs=Multiply(A=@n50, B=ScalarParameter() {defaultValue=1.5, group=Cube Queen Emissive, parameterName=SecondaryEmissiveTiling})
   Bias=ScalarParameter() {defaultValue=3.5, group=Cube Queen Emissive, parameterName=SecondaryEmissiveMipBias}
   Apply View MipBias
  ) {
   group=Cube Queen Emissive
   mipValueMode=TMVM_MipBias
   parameterName=NoiseB
   samplerType=SAMPLERTYPE_LinearGrayscale
   texture=/Game/Characters/Player/Female/Medium/Bodies/F_MED_Cube_Queen/Textures/TechArt/T_CubeQueenNoiseB
  }
  NamedRerouteDeclaration(None=MaterialFunctionCall!(
    NoiseSpeed=ScalarParameter() {defaultValue=0.2, group=Cube Queen Emissive, parameterName=SecondaryEmissiveNoiseSpeed}
    GradientTexture=@n51.RGB
   ) {materialFunction=&/MF_Cube_QueenPlasma}.Result) {name=SecondaryEmissives, nodeColor=<rgba 0.7344 1 0 1>}
  NamedRerouteDeclaration(None=@n11.Fresnel) {name=Fresnel_IN, nodeColor=<rgba 0 1 0.7891 1>}
  n52 = Divide(A=Subtract(A=NamedRerouteUsage(), B=Constant3Vector() {constant=<rgba -58 -24 0 0>}), B=Constant3Vector() {constant=<rgba 116 64 182 0>})
  n53 = ComponentMask(None=@n52) {b=true}
  NamedRerouteDeclaration(None=@n53) {name=VerticalGradient, nodeColor=<rgba 0 1 1 1>}
  n54 = NamedRerouteUsage()
  n55 = Abs(None=ComponentMask(None=@n54) {r=true})
  n56 = ComponentMask(None=@n54) {b=true, g=true}
  NamedRerouteDeclaration(None=SphereMask(A=AppendVector(A=@n55, B=@n56), B=Constant3Vector() {constant=<rgba 20 -7 135 0>}, Radius=Constant() {r=13}, Hardness=Constant() {r=1})) {name=ShouldersMask, nodeColor=<rgba 0.5469 0 1 1>}
  NamedRerouteDeclaration(None=VertexInterpolator(VS=PreSkinnedPosition()).PS) {name=Interpolated_preskinnedPos, nodeColor=<rgba 1 0 0.8906 1>}
  n57 = NamedRerouteUsage()
  n58 = ComponentMask(None=@n57) {r=true}
  n59 = SphereMask(
   A=AppendVector(A=Abs(None=@n58), B=ComponentMask(None=@n57) {g=true})
   B=Constant2Vector() {g=5, r=15}
   Radius=Constant() {r=10}
   Hardness=Constant() {r=1}
  )
  n60 = Step(Y=ComponentMask(None=@n57) {b=true}) {constX=50}
  n61 = Step(Y=ComponentMask(None=@n57) {b=true, g=true}, X=Constant2Vector() {g=95})
  n62 = ComponentMask(None=@n61) {r=true}
  n63 = ComponentMask(None=@n61) {g=true}
  n64 = Saturate(None=Add(A=Multiply(A=@n59, B=@n60), B=Multiply(A=@n62, B=@n63)))
  NamedRerouteDeclaration(None=@n64) {name=BackTailsMask, nodeColor=<rgba 0 0.6719 1 1>}
  NamedRerouteDeclaration(None=StaticComponentMaskParameter(None=@n13) {defaultR=true, group=Anodized Metal, parameterName=AnodizedMetal Channel}) {name=AnodizedMetalChannel, nodeColor=<rgba 1 0.1172 0 1>}
  n65 = Step(Y=Abs(None=ComponentMask(None=@n57) {r=true}), X=Constant() {r=7})
  n66 = Saturate(None=Subtract(A=Multiply(A=@n63, B=@n65), B=@n62))
  NamedRerouteDeclaration(None=Saturate(None=Add(A=@n64, B=@n66))) {name=AllTailsMask, nodeColor=<rgba 1 0 0.7734 1>}
  n67 = VertexColor()
  n68 = CurveAtlasRowParameter(CurveTime=@n67.R) {
   atlas=/Game/Characters/Shared/CurveAtlas/AthenaCharacterCurveAtlas
   curve=/Game/Characters/Shared/CurveAtlas/CQ_FloatingPiecesVectorLUT
   defaultValue=26
   group=Cube Queen Emissive
   parameterName=VectorLUT
  }
  n69 = Ceil(None=@n67.R)
  NamedRerouteDeclaration(None=Multiply(A=@n68, B=@n69)) {name=WPO_VectorDebug, nodeColor=<rgba 0 0.1562 1 1>}
  n70 = Sign(None=PreSkinnedPosition())
  n71 = Multiply(A=Multiply(A=@n68.A, B=Normalize(VectorInput=@n68)), B=AppendVector(A=ComponentMask(None=@n70) {r=true}, B=Constant2Vector() {g=1, r=1}))
  n72 = Subtract(A=Time(), B=ScalarParameter() {group=Cube Queen WPO, parameterName=GameTimeFromBP})
  n73 = Add(A=ScalarParameter() {defaultValue=0.3, group=Cube Queen WPO, parameterName=Animation_Duration}) {constB=0}
  n74 = ScalarParameter() {defaultValue=3, group=Cube Queen WPO, parameterName=Animation_EaseInOut}
  n75 = MaterialFunctionCall(In=Saturate(None=Divide(A=@n72, B=@n73)), Power=@n74) {materialFunction=&SCurve}
  n76 = OneMinus(None=@n75.Result)
  n77 = LinearInterpolate(A=@n76, B=@n75.Result, Alpha=ScalarParameter() {group=Cube Queen WPO, parameterName=ToggleFromBP})
  n78 = LinearInterpolate(
   A=ScalarParameter() {defaultValue=10, group=ActivationSwitch, parameterName=WPODeployPiecesAmplitude}
   B=ScalarParameter() {group=ActivationSwitch, parameterName=WPODeployPiecesAmplitude_ADS}
   Alpha=@n77
  )
  n79 = ScalarParameter() {defaultValue=0.25, group=Cube Queen WPO, parameterName=WPOFloatingMotionSpeed}
  n80 = Add(A=Time(), B=@n67.R)
  n81 = Multiply(
   A=Sine(None=Multiply(A=@n79, B=@n80))
   B=LinearInterpolate(
    A=ScalarParameter() {defaultValue=1, group=Cube Queen WPO, parameterName=WPOFloatingMotionAmplitude}
    B=ScalarParameter() {defaultValue=0.3, group=Cube Queen WPO, parameterName=WPOFloatingMotionAmplitude_ADS}
    Alpha=@n77
   )
  )
  n82 = Add(
   A=Multiply(A=@n67.R, B=ScalarParameter() {defaultValue=5, parameterName=WPOFloatingMotionSeedMultiplier})
   B=Multiply(A=Time(), B=ScalarParameter() {defaultValue=0.5, group=Cube Queen WPO, parameterName=WPOFloatingMotionDirectionChangeSpeed})
  )
  n83 = Normalize(VectorInput=MaterialFunctionCall(Hue Shift Percentage=Frac(None=@n82), Texture=Constant3Vector() {constant=<rgba 0.2095 0.9323 0.5624 0>}) {materialFunction=&HueShift}.Result)
  n84 = Multiply(A=@n69, B=Add(A=Multiply(A=@n71, B=@n78), B=Multiply(A=@n81, B=@n83)))
  NamedRerouteDeclaration(None=Transform(None=MaterialFunctionCall!(InVector=@n84) {materialFunction=/Game/Athena/Environments/Landscape/Foliage/MF_MakePreciseVec3}.Result) {transformSourceType=TRANSFORMSOURCE_Local}) {name=WPO_FloatingPieces, nodeColor=<rgba 0.4297 1 0 1>}
  n85 = ScalarParameter() {defaultValue=0.2, group=Cube Queen Emissive, parameterName=NoiseSpeed}
  n86 = ScalarParameter() {defaultValue=3, group=Cube Queen Emissive, parameterName=UV_TilingGlobal}
  n87 = StaticSwitchParameter(True=ComponentMask(None=NamedRerouteUsage()) {g=true, r=true}, False=TextureCoordinate()) {group=Cube Queen Emissive, parameterName=UseSimplePlanarCoords}
  n88 = Constant3Vector() {constant=<rgba -1 1 0 0>}
  n89 = NamedRerouteUsage()
  n90 = MaterialFunctionCall(f2=Multiply(A=ComponentMask(None=@n88) {g=true, r=true}, B=ComponentMask(None=@n89) {g=true, r=true})) {materialFunction=&AddComponents}
  n91 = ComponentMask(None=@n89) {b=true}
  n92 = LinearInterpolate(
   A=@n50
   B=Multiply(A=AppendVector(A=@n90.f2, B=@n91), B=ScalarParameter() {defaultValue=0.005, group=Cube Queen Emissive, parameterName=UV_TilingShoulders})
   Alpha=NamedRerouteUsage()
  )
  n93 = LinearInterpolate(A=@n86, B=ScalarParameter() {defaultValue=0.5, group=Cube Queen Emissive, parameterName=UV_TilingTails}, Alpha=NamedRerouteUsage())
  n94 = Constant()
  n95 = Constant() {r=-0.001}
  n96 = Add(
   A=StaticSwitchParameter(True=Multiply(A=@n86, B=@n87), False=Multiply(A=@n92, B=@n93)) {group=Cube Queen Emissive, parameterName=UseSimpleUVCoords}
   B=Multiply(A=Time(), B=AppendVector(A=@n94, B=@n95))
  )
  n97 = TextureSampleParameter2D!(UVs=Multiply(A=@n96, B=Constant() {r=5}), Apply View MipBias) {
   group=Cube Queen Emissive
   parameterName=NoiseB
   samplerType=SAMPLERTYPE_LinearGrayscale
   texture=&T_CubeQueenNoiseB
  }
  n98 = TextureSampleParameter2D(UVs=Multiply(A=Constant() {r=2}, B=@n96)) {
   group=Cube Queen Emissive
   parameterName=NoiseA
   samplerType=SAMPLERTYPE_LinearGrayscale
   texture=/Game/Characters/Player/Female/Medium/Bodies/F_MED_Cube_Queen/Textures/TechArt/T_CubeQueenNoiseA
  }
  n99 = ScalarParameter() {defaultValue=0.26, group=Cube Queen Emissive, parameterName=NoiseBlendPos}
  n100 = ScalarParameter() {defaultValue=0.08, group=Cube Queen Emissive, parameterName=NoiseBlendWidth}
  n101 = Saturate(None=InverseLinearInterpolate(A=@n99, B=Add(A=@n99, B=@n100), Value=@n53))
  n102 = ComponentMask(None=@n52) {b=true, r=true}
  n103 = Constant3Vector() {constant=<rgba 1.5 0.5 0 0>}
  n104 = Constant3Vector() {constant=<rgba 0.5 0.37 0 1>}
  n105 = Subtract(
   A=@n101
   B=SphereMask(
    A=MaterialFunctionCall(UVs=@n102, Texture Scale=@n103) {materialFunction=&ScaleUVsByCenter}.UVs
    B=ComponentMask(None=@n104) {g=true, r=true}
    Radius=Constant() {r=0.15}
    Hardness=Constant() {r=0.3}
   )
  )
  n106 = SphereMask(
   A=ComponentMask(None=@n52) {b=true, r=true}
   B=ComponentMask(None=Constant3Vector() {constant=<rgba 0.5 0.753 0 1>}) {g=true, r=true}
   Radius=Constant() {r=0.01}
   Hardness=Constant()
  )
  n107 = ComponentMask(None=@n52) {b=true, r=true}
  n108 = ComponentMask(None=VectorParameter() {defaultValue=<rgba 0.5 0.65 0 1>}.RGB) {g=true, r=true}
  n109 = Constant() {r=0.025}
  n110 = Constant()
  n111 = Saturate(None=Subtract(A=Subtract(A=@n105, B=@n106), B=SphereMask(A=@n107, B=@n108, Radius=@n109, Hardness=@n110)))
  n112 = ShadingPathSwitch(Default=@n111, Mobile=@n101)
  n113 = LinearInterpolate(
   A=Power(Base=@n97.R, Exp=ScalarParameter() {defaultValue=1, group=Cube Queen Emissive, parameterName=NoiseBPow})
   B=Power(Base=@n98.R, Exp=ScalarParameter() {defaultValue=0.7, group=Cube Queen Emissive, parameterName=NoiseAPow})
   Alpha=StaticSwitchParameter(
    True=ScalarParameter() {group=Cube Queen Emissive, parameterName=NoiseBlend}
    False=StaticSwitchParameter(True=@n112, False=@n111) {parameterName=useMobileOptims}
   ) {group=Cube Queen Emissive, parameterName=UseSimpleUVCoords}
  )
  n114 = Multiply(
   A=NamedRerouteUsage()
   B=Multiply(
    A=MaterialFunctionCall(NoiseSpeed=@n85, GradientTexture=@n113) {materialFunction=&MF_Cube_QueenPlasma}.Result
    B=ScalarParameter() {defaultValue=0.75, group=Cube Queen Emissive, parameterName=NoiseEmissiveStrength}
   )
  ) {constA=1}
  n115 = ScalarParameter() {group=Cube Queen Emissive, parameterName=TailsGradientPos}
  n116 = Add(A=@n115, B=ScalarParameter() {defaultValue=0.6, group=Cube Queen Emissive, parameterName=TailsGradienWidth})
  n117 = NamedRerouteUsage()
  n118 = OneMinus(None=Saturate(None=SmoothStep(Min=@n115, Max=@n116, Value=@n117)))
  n119 = ScalarParameter() {defaultValue=2, group=Cube Queen Emissive, parameterName=TailsTipBrightness}
  n120 = NamedRerouteUsage()
  n121 = Add(
   A=Step(X=ComponentMask(None=@n120) {b=true}) {constY=1}
   B=ScalarParameter() {defaultValue=0.3, group=Cube Queen Emissive, parameterName=SolesBrightness}
  )
  n122 = LinearInterpolate(A=Constant() {r=1}, B=Multiply(A=Multiply(A=@n118, B=@n119), B=Saturate(None=@n121)), Alpha=NamedRerouteUsage())
  NamedRerouteDeclaration(None=StaticSwitchParameter(True=@n114, False=Multiply(A=@n114, B=@n122)) {group=Cube Queen Emissive, parameterName=UseSimpleUVCoords}) {name=CubeQueenEmissives, nodeColor=<rgba 1 0.1875 0 1>}
  NamedRerouteUsage()
  Power(Base=@n75.Result, Exp=@n74) {constExponent=8}
  Power(Base=@n76, Exp=@n74) {constExponent=8}
  PreSkinnedLocalBounds()
  SetMaterialAttributes!(MaterialAttributes, Emissive Color=NamedRerouteUsage()) {attributeSetTypes [#2]}
 ]
}
